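# Unlit translucent particle material for light rays in the Outlander "In The Zone" effect, usable on Niagara and Cascade sprites.
# Two channels of one ray texture pan in opposite X directions and are summed, with tiling and speed set by dynamic parameters.
# A powered radial gradient masks the rays into opacity, softened by depth fade and scaled by particle alpha.
# Emissive blends two color parameters, optionally tinted by particle color, through an emissive adjustment function.
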Material /Game/Effects/Fort_Effects/Materials/Outlander/M_InTheZone_Lightrays_01
graph {
 properties {
  bUsedWithNanite=true
  bUsedWithNiagaraSprites=true
  bUsedWithParticleSprites=true
  blendMode=BLEND_Translucent
  pixelDepthOffsetMode=PDOM_Legacy
  shadingModel=MSM_Unlit
  usedShadingModels=Unlit
 }
 nodes [
  n1 = TextureCoordinate()
  n2 = DynamicParameter() {paramNames [UV A, Time A, UV B, Time B]}
  n3 = Multiply(A=@n1, B=@n2.UV A)
  n4 = Time()
  n5 = Multiply(A=@n4, B=@n2.Time A)
  n6 = Multiply(A=@n1, B=@n2.UV B)
  n7 = Multiply(A=@n4, B=@n2.Time B)
  n8 = Add(
   A=TextureSample(UVs=Panner(Coordinate=@n3, Time=@n5) {speedX=-0.18}) {texture=/Game/Effects/Fort_Effects/Textures/Outlander/T_InTheZone_LightRays_01}.R
   B=TextureSample(UVs=Panner(Coordinate=@n6, Time=@n7) {speedX=0.23}) {texture=&T_InTheZone_LightRays_01}.G
  )
  Divide(A=@n8)
  n9 = ParticleColor()
  n10 = LinearInterpolate(
   A=VectorParameter() {parameterName=Color A, defaultValue=<rgba 1 0.5534 0.295 0>}.RGB
   B=VectorParameter() {parameterName=Color B, defaultValue=<rgba 0.401 0.3782 1 0>}.RGB
  )
  MaterialFunctionCall(Emissive=LinearInterpolate(A=Multiply(A=@n9.RGB, B=@n10), B=@n10, Alpha=ScalarParameter() {parameterName=0ForParticleColor, sliderMax=1})) {materialFunction=&MF_FX_EmissiveAdjustments}
  n11 = Power(
   Base=MaterialFunctionCall!(UVs, CenterPosition, Radius=Constant() {r=0.5}) {materialFunction=/Engine/Functions/Engine_MaterialFunctions01/Gradient/SphereGradient-2D}.Result Radius
   Exp=ScalarParameter() {defaultValue=3, parameterName=Exponent}
  )
  Multiply(A=@n9.A, B=DepthFade(Opacity=Multiply(A=@n11, B=@n8), FadeDistance=ScalarParameter() {defaultValue=40, parameterName=Depth Fade Distance}))
 ]
}
